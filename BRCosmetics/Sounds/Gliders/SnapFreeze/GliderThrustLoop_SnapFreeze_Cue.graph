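# Looping thrust sound cue for the SnapFreeze glider, using glider thrust attenuation and cosmetic glider loop audio settings.
# A mixer blends seven looping layers: low, main, two tonal, turbine, turbine wind and wind.
# Each layer's continuous modulator reads the AnyDirection parameter to scale pitch and volume.
# Pitch rises slightly with the parameter, and volume rises on every layer except the low one, which fades out.
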SoundCue /BRCosmetics/Sounds/Gliders/SnapFreeze/GliderThrustLoop_SnapFreeze_Cue
root=Mixer_1
nodes [
 {
  id=Mixer_1
  children [
   ModulatorContinuous_1
   ModulatorContinuous_2
   ModulatorContinuous_3
   ModulatorContinuous_4
   ModulatorContinuous_5
   ModulatorContinuous_6
   ModulatorContinuous_8
  ]
  properties {inputVolume [0.9,0.6,0.5,0.9,0.6,0.85,0.85]}
 }
 {
  id=ModulatorContinuous_1
  children [WavePlayer_6]
  properties {
   pitchModulationParams {default=0, maxOutput=1.15, minOutput=1, parameterName=AnyDirection}
   volumeModulationParams {default=0, maxOutput=0, minOutput=0.5, parameterName=AnyDirection}
  }
 }
 {
  id=ModulatorContinuous_2
  children [WavePlayer_1]
  properties {
   pitchModulationParams {default=0, maxOutput=1.1, minOutput=1, parameterName=AnyDirection}
   volumeModulationParams {default=0, maxOutput=1.2, minOutput=0.5, parameterName=AnyDirection}
  }
 }
 {
  id=ModulatorContinuous_3
  children [WavePlayer_8]
  properties {
   pitchModulationParams {default=0, maxOutput=1.1, minOutput=1, parameterName=AnyDirection}
   volumeModulationParams {default=0, maxOutput=0.5, minOutput=0.3, parameterName=AnyDirection}
  }
 }
 {
  id=ModulatorContinuous_4
  children [WavePlayer_7]
  properties {
   pitchModulationParams {default=0, maxOutput=1.15, minOutput=1, parameterName=AnyDirection}
   volumeModulationParams {default=0, maxOutput=0.65, minOutput=0.3, parameterName=AnyDirection}
  }
 }
 {
  id=ModulatorContinuous_5
  children [WavePlayer_3]
  properties {
   pitchModulationParams {default=0, maxOutput=1.1, minOutput=1, parameterName=AnyDirection}
   volumeModulationParams {default=0, minOutput=0.8, parameterName=AnyDirection}
  }
 }
 {
  id=ModulatorContinuous_6
  children [WavePlayer_4]
  properties {
   pitchModulationParams {default=0, maxOutput=1.1, minOutput=1, parameterName=AnyDirection}
   volumeModulationParams {default=0, maxOutput=0.8, minOutput=0.4, parameterName=AnyDirection}
  }
 }
 {
  id=ModulatorContinuous_8
  children [WavePlayer_15]
  properties {
   pitchModulationParams {default=0, maxOutput=1.15, minOutput=1, parameterName=AnyDirection}
   volumeModulationParams {default=0, maxOutput=0.5, minOutput=0.2, parameterName=AnyDirection}
  }
 }
 {id=WavePlayer_1, properties {bLooping=true, soundWaveAssetPtr=&/IP_Glider_SnapFreeze_Thrust_LP_Tonal}}
 {
  id=WavePlayer_15
  properties {bLooping=true, soundWaveAssetPtr=&/IP_Glider_SnapFreeze_Thrust_LP_Tonal_High}
 }
 {id=WavePlayer_3, properties {bLooping=true, soundWaveAssetPtr=&/IP_Glider_SnapFreeze_Thrust_LP_Turbine}}
 {
  id=WavePlayer_4
  properties {bLooping=true, soundWaveAssetPtr=&/IP_Glider_SnapBreeze_Thrust_LP_Turbine_Wind}
 }
 {id=WavePlayer_6, properties {bLooping=true, soundWaveAssetPtr=&/IP_Glider_SnapFreeze_Thrust_LP_Low}}
 {id=WavePlayer_7, properties {bLooping=true, soundWaveAssetPtr=&/IP_Glider_SnapFreeze_Thrust_LP_Main}}
 {id=WavePlayer_8, properties {bLooping=true, soundWaveAssetPtr=&/IP_Glider_SnapFreeze_Thrust_LP_Wind}}
]
properties {
 attenuationSettings=&Fort_Player_Glider_Thrust_Athena_Att
 audioPropertiesSheet=&APS_Cosmetic_Glider_Loop
 maxDistance=12500
 soundClassObject=&Fort_Gliders_Land
 volumeMultiplier=0.8
 concurrencySet [null]
}
